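# Decorative frontend UI widget that shows a fake stats graph of five thin bars stacked in a fixed-width vertical box.
# Each bar is an Image using the M_UI_FakeBarGraphsStats material.
# On Construct, or PreConstruct at design time for editor preview, three functions run in sequence on each bar's dynamic material.
# They randomize its time offset, bar offset amount and tiling amount so the bars differ from one another.
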WidgetBlueprint /BRCosmetics/Athena/UI/Frontend/Spatial/Art/FakeStatsGraphs
graphs [
 EventGraph {nodes [
   n1 = Event PreConstruct()
   > n2 = Branch(Condition=@n1.IsDesignTime)
   RandomizePositions(execute=[@n3.then, @n2.then])
   > RandomizeTimings()
   > RandomizeTiling()
   n3 = Event Construct()
  ]}
 RandomizePositions {nodes [
   RandomizePositions() {is=title K2Node_FunctionEntry}
   > n1 = GetDynamicMaterial(self=Get Bar_1() -> (Bar_1:Image Object Reference))
   > SetScalarParameterValue(self=@n1, ParameterName=RandomTimeOffset, Value=RandomFloatInRange(Min=0, Max=0.8))
   > n2 = GetDynamicMaterial(self=Get Bar_2() -> (Bar_2:Image Object Reference))
   > SetScalarParameterValue(self=@n2, ParameterName=RandomTimeOffset, Value=RandomFloatInRange(Min=0, Max=0.8))
   > n3 = GetDynamicMaterial(self=Get Bar_3() -> (Bar_3:Image Object Reference))
   > SetScalarParameterValue(self=@n3, ParameterName=RandomTimeOffset, Value=RandomFloatInRange(Min=0, Max=0.8))
   > n4 = GetDynamicMaterial(self=Get Bar_4() -> (Bar_4:Image Object Reference))
   > SetScalarParameterValue(self=@n4, ParameterName=RandomTimeOffset, Value=RandomFloatInRange(Min=0, Max=0.8))
   > n5 = GetDynamicMaterial(self=Get Bar_5() -> (Bar_5:Image Object Reference))
   > SetScalarParameterValue(self=@n5, ParameterName=RandomTimeOffset, Value=RandomFloatInRange(Min=0, Max=0.8))
  ]}
 RandomizeTiling {nodes [
   RandomizeTiling() {is=title K2Node_FunctionEntry}
   > n1 = GetDynamicMaterial(self=Get Bar_1())
   > SetScalarParameterValue(self=@n1, ParameterName=TilingAmount, Value=RandomFloatInRange(Min=10, Max=20))
   > n2 = GetDynamicMaterial(self=Get Bar_2())
   > SetScalarParameterValue(self=@n2, ParameterName=TilingAmount, Value=RandomFloatInRange(Min=10, Max=20))
   > n3 = GetDynamicMaterial(self=Get Bar_3())
   > SetScalarParameterValue(self=@n3, ParameterName=TilingAmount, Value=RandomFloatInRange(Min=10, Max=20))
   > n4 = GetDynamicMaterial(self=Get Bar_4())
   > SetScalarParameterValue(self=@n4, ParameterName=TilingAmount, Value=RandomFloatInRange(Min=10, Max=20))
   > n5 = GetDynamicMaterial(self=Get Bar_5())
   > SetScalarParameterValue(self=@n5, ParameterName=TilingAmount, Value=RandomFloatInRange(Min=10, Max=20))
  ]}
 RandomizeTimings {nodes [
   RandomizeTimings() {is=title K2Node_FunctionEntry}
   > n1 = GetDynamicMaterial(self=Get Bar_1())
   > SetScalarParameterValue(self=@n1, ParameterName=BarOffset_Amount, Value=RandomFloatInRange(Min=40, Max=80))
   > n2 = GetDynamicMaterial(self=Get Bar_2())
   > SetScalarParameterValue(self=@n2, ParameterName=BarOffset_Amount, Value=RandomFloatInRange(Min=40, Max=80))
   > n3 = GetDynamicMaterial(self=Get Bar_3())
   > SetScalarParameterValue(self=@n3, ParameterName=BarOffset_Amount, Value=RandomFloatInRange(Min=40, Max=80))
   > n4 = GetDynamicMaterial(self=Get Bar_4())
   > SetScalarParameterValue(self=@n4, ParameterName=BarOffset_Amount, Value=RandomFloatInRange(Min=40, Max=80))
   > n5 = GetDynamicMaterial(self=Get Bar_5())
   > SetScalarParameterValue(self=@n5, ParameterName=BarOffset_Amount, Value=RandomFloatInRange(Min=40, Max=80))
  ]}
]
variables []
widgets {
 info {inheritedWidgetCount=0, parentClass=&UserWidget, rootWidgetClass=&SizeBox, widgetCount=11}
 widgets [
  SizeBox_0 {widgetClassPath=&SizeBox, widget=SizeBox_0, properties {bOverride_WidthOverride=true, widthOverride=88}}
  VerticalBox_8 {widgetClassPath=&VerticalBox, parent=SizeBox_0, widget=VerticalBox_8, slot {class=SizeBoxSlot}}
  Bar_1 {
   widgetClassPath=&Image
   bIsVariable=true
   parent=VerticalBox_8
   widget=Bar_1
   properties {brush {
     bIsDynamicallyLoaded=false
     drawAs=Image
     imageType=NoImage
     mirroring=NoMirror
     resourceName=None
     resourceObject=&/M_UI_FakeBarGraphsStats
     tiling=NoTile
     imageSize=<32 4>
     margin=<ltrb 0 0 0 0>
     outlineSettings {
      bUseBrushTransparency=false
      roundingType=HalfHeightRadius
      width=0
      color {colorUseRule=UseColor_Specified, specifiedColor=<rgba 0 0 0 0>}
      cornerRadii=<0 0 0 0>
     }
     tintColor {colorUseRule=UseColor_Specified, specifiedColor=<rgba 1 1 1 1>}
     uVRegion {bIsValid=false, max=<0 0>, min=<0 0>}
    }}
   slot {class=VerticalBoxSlot}
  }
  Spacer_33 {widgetClassPath=&Spacer, parent=VerticalBox_8, widget=Spacer_33, properties {size=<1 4>}, slot {class=VerticalBoxSlot}}
  Bar_2 {
   widgetClassPath=&Image
   bIsVariable=true
   parent=VerticalBox_8
   widget=Bar_2
   properties {brush {
     bIsDynamicallyLoaded=false
     drawAs=Image
     imageType=NoImage
     mirroring=NoMirror
     resourceName=None
     resourceObject=&M_UI_FakeBarGraphsStats
     tiling=NoTile
     imageSize=<32 4>
     margin=<ltrb 0 0 0 0>
     outlineSettings {
      bUseBrushTransparency=false
      roundingType=HalfHeightRadius
      width=0
      color {colorUseRule=UseColor_Specified, specifiedColor=<rgba 0 0 0 0>}
      cornerRadii=<0 0 0 0>
     }
     tintColor {colorUseRule=UseColor_Specified, specifiedColor=<rgba 1 1 1 1>}
     uVRegion {bIsValid=false, max=<0 0>, min=<0 0>}
    }}
   slot {class=VerticalBoxSlot}
  }
  Spacer {widgetClassPath=&Spacer, parent=VerticalBox_8, widget=Spacer, properties {size=<1 4>}, slot {class=VerticalBoxSlot}}
  Bar_3 {
   widgetClassPath=&Image
   bIsVariable=true
   parent=VerticalBox_8
   widget=Bar_3
   properties {brush {
     bIsDynamicallyLoaded=false
     drawAs=Image
     imageType=NoImage
     mirroring=NoMirror
     resourceName=None
     resourceObject=&M_UI_FakeBarGraphsStats
     tiling=NoTile
     imageSize=<32 4>
     margin=<ltrb 0 0 0 0>
     outlineSettings {
      bUseBrushTransparency=false
      roundingType=HalfHeightRadius
      width=0
      color {colorUseRule=UseColor_Specified, specifiedColor=<rgba 0 0 0 0>}
      cornerRadii=<0 0 0 0>
     }
     tintColor {colorUseRule=UseColor_Specified, specifiedColor=<rgba 1 1 1 1>}
     uVRegion {bIsValid=false, max=<0 0>, min=<0 0>}
    }}
   slot {class=VerticalBoxSlot}
  }
  Spacer_1 {widgetClassPath=&Spacer, parent=VerticalBox_8, widget=Spacer_1, properties {size=<1 4>}, slot {class=VerticalBoxSlot}}
  Bar_4 {
   widgetClassPath=&Image
   bIsVariable=true
   parent=VerticalBox_8
   widget=Bar_4
   properties {brush {
     bIsDynamicallyLoaded=false
     drawAs=Image
     imageType=NoImage
     mirroring=NoMirror
     resourceName=None
     resourceObject=&M_UI_FakeBarGraphsStats
     tiling=NoTile
     imageSize=<32 4>
     margin=<ltrb 0 0 0 0>
     outlineSettings {
      bUseBrushTransparency=false
      roundingType=HalfHeightRadius
      width=0
      color {colorUseRule=UseColor_Specified, specifiedColor=<rgba 0 0 0 0>}
      cornerRadii=<0 0 0 0>
     }
     tintColor {colorUseRule=UseColor_Specified, specifiedColor=<rgba 1 1 1 1>}
     uVRegion {bIsValid=false, max=<0 0>, min=<0 0>}
    }}
   slot {class=VerticalBoxSlot}
  }
  Spacer_2 {widgetClassPath=&Spacer, parent=VerticalBox_8, widget=Spacer_2, properties {size=<1 4>}, slot {class=VerticalBoxSlot}}
  Bar_5 {
   widgetClassPath=&Image
   bIsVariable=true
   parent=VerticalBox_8
   widget=Bar_5
   properties {brush {
     bIsDynamicallyLoaded=false
     drawAs=Image
     imageType=NoImage
     mirroring=NoMirror
     resourceName=None
     resourceObject=&M_UI_FakeBarGraphsStats
     tiling=NoTile
     imageSize=<32 4>
     margin=<ltrb 0 0 0 0>
     outlineSettings {
      bUseBrushTransparency=false
      roundingType=HalfHeightRadius
      width=0
      color {colorUseRule=UseColor_Specified, specifiedColor=<rgba 0 0 0 0>}
      cornerRadii=<0 0 0 0>
     }
     tintColor {colorUseRule=UseColor_Specified, specifiedColor=<rgba 1 1 1 1>}
     uVRegion {bIsValid=false, max=<0 0>, min=<0 0>}
    }}
   slot {class=VerticalBoxSlot}
  }
 ]
}
