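# Left panel widget of the Battle Pass door, based on FortEventLevelNavigationWidget, showing a sheared "Level" header and number above two FakeStatsGraphs decorations.
# InitLevel writes the level number text.
# SetDoorPanelStyle rebuilds both fonts with purchased or unpurchased fill and outline materials and tints the graphs to match.
# SetStyleUnpurchased instead recolors dynamic font material parameters and graphs; neither style function is called within this widget.
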WidgetBlueprint /BRCosmetics/Athena/UI/Frontend/Spatial/Art/BP_Door/BattlePass_Door_LeftPanel
graphs [
 EventGraph {nodes [
   n1 = InitLevel!() -> (OutputDelegate, then, Level:int) {is=title K2Node_CustomEvent}
   > "SetText (Text)"(self=Get LevelNumber() -> (LevelNumber:Common Text Object Reference), InText="To Text (Integer)"(Value=@n1.Level))
   Event PreConstruct()
   Event Construct()
   Event Tick()
  ]}
 SetDoorPanelStyle {nodes [
   n1 = SetDoorPanelStyle!() -> (then, Purchased:bool=~) {is=title K2Node_FunctionEntry}
   n2 = MakeLiteralInt!(Value:int=13) -> (ReturnValue:int=0)
   n3 = Select(
    Option 0:Object Reference=&/M_UI_BPDoor_Panel_Text_Outline_NotPurchased
    Option 1:Object Reference=&/M_UI_BPDoor_Panel_Text_Outline_Purchased
    Index=@n1
   ) -> (ReturnValue:Object Reference)
   > SetFont(
    self=Get LevelHeaderText() -> (LevelHeaderText:Common Text Object Reference)
    InFontInfo=MakeSlateFontInfo!(
     FontObject:Object Reference=&BurbankSmall
     FontMaterial:Object Reference=Select(
      Option 0=&/M_UI_BPDoor_Panel_Text_Header_NotPurchased
      Option 1=&/M_UI_BPDoor_Panel_Text_Header_Purchased
      Index=@n1
     )
     OutlineSettings:Font Outline Settings Structure=Make Font Outline Settings!(
      OutlineSize:int=2
      bMiteredCorners:bool=False
      bSeparateFillAlpha:bool=true
      bApplyOutlineToDropShadows:bool=False
      OutlineMaterial:Object Reference=@n3
      OutlineColor:Linear Color Structure=(R=1,G=1,B=1,A=1)
     ) -> (FontOutlineSettings:Font Outline Settings Structure) {is=title K2Node_MakeStruct}
     TypefaceFontName:Name=burbanksmall-black
     Size:float="To Float (Integer)"(InInt=@n2)
     LetterSpacing:int=0
     SkewAmount:float=0
     bForceMonospaced:bool=false
     bMaterialIsStencil:bool=false
     MonospacedWidth:float=1
    ) -> (ReturnValue:Slate Font Info Structure)
   )
   n4 = MakeLiteralInt(Value=56) -> (ReturnValue=0)
   n5 = Select(
    Option 0=&M_UI_BPDoor_Panel_Text_Outline_NotPurchased
    Option 1=&M_UI_BPDoor_Panel_Text_Outline_Purchased
    Index=@n1
   )
   > SetFont(
    self=Get LevelNumber()
    InFontInfo=MakeSlateFontInfo(
     FontObject=&Burbank
     FontMaterial=Select(
      Option 0=&/M_UI_BPDoor_Panel_Text_NotPurchased
      Option 1=&/M_UI_BPDoor_Panel_Text_Purchased
      Index=@n1
     )
     OutlineSettings=Make Font Outline Settings(
      OutlineSize=2
      bMiteredCorners=False
      bSeparateFillAlpha=true
      bApplyOutlineToDropShadows=False
      OutlineMaterial=@n5
      OutlineColor=(R=1,G=1,B=1,A=1)
     )
     TypefaceFontName=bankBigRegular-Black
     Size="To Float (Integer)"(InInt=@n4)
     LetterSpacing=0
     SkewAmount=0
     bForceMonospaced=false
     bMaterialIsStencil=false
     MonospacedWidth=1
    )
   )
   > SetColorAndOpacity(
    self:User Widget Object Reference=Get FakeStatsGraphs_Top() -> (FakeStatsGraphs_Top:Fake Stats Graphs Object Reference)
    InColorAndOpacity:Linear Color Structure=Select(Option 0:Linear Color Structure=(R=0.25,G=0.625,B=1,A=0.125), Option 1:Linear Color Structure=(R=1,G=0.7011,B=0.0999,A=0.175), Index=@n1) -> (ReturnValue:Linear Color Structure)
   )
   > SetColorAndOpacity(
    self=Get FakeStatsGraphs_Bottom() -> (FakeStatsGraphs_Bottom:Fake Stats Graphs Object Reference)
    InColorAndOpacity=Select(Option 0=(R=0.2,G=0.5333,B=1,A=0.075), Option 1=(R=1,G=0.552,B=0.0999,A=0.1), Index=@n1)
   )
  ]}
 SetStyleUnpurchased {nodes [
   SetStyleUnpurchased() {is=title K2Node_FunctionEntry}
   > n1 = GetDynamicFontMaterial!(execute, self:Text Object Reference=Get LevelHeaderText()) -> (then, ReturnValue:Material Instance Dynamic Object Reference)
   > SetVectorParameterValue(self=@n1, ParameterName=DotsColor, Value=(R=0.375,G=0.5,B=0.75,A=1))
   > SetVectorParameterValue(self=@n1, ParameterName=Squares_Color_A, Value=(R=0.2502,G=0.2918,B=0.5029,A=1))
   > SetVectorParameterValue(self=@n1, ParameterName=Squares_Color_B, Value=(R=0.325,G=0.3688,B=0.5,A=1))
   > n2 = GetDynamicFontMaterial(self=Get LevelNumber())
   > SetVectorParameterValue(self=@n2, ParameterName=DotsColor, Value=(R=0.24,G=0.4733,B=0.8,A=1))
   > SetVectorParameterValue(self=@n2, ParameterName=Squares_Color_A, Value=(R=0.2502,G=0.2918,B=0.5029,A=1))
   > SetVectorParameterValue(self=@n2, ParameterName=Squares_Color_B, Value=(R=0.325,G=0.3688,B=0.5,A=1))
   > SetColorAndOpacity(self=Get FakeStatsGraphs_Top(), InColorAndOpacity=(R=0.2502,G=0.2918,B=0.5029,A=0.25))
   > SetColorAndOpacity(self=Get FakeStatsGraphs_Bottom(), InColorAndOpacity=(R=0.2502,G=0.2918,B=0.5029,A=0.125))
  ]}
]
variables []
widgets {
 info {inheritedWidgetCount=1, parentClass=/Script/FortniteGame.FortEventLevelNavigationWidget, rootWidgetClass=&SizeBox, widgetCount=11}
 widgets [
  SizeBox_1 {
   widgetClassPath=&SizeBox
   widget=SizeBox_1
   properties {bOverride_WidthOverride=true, maxDesiredWidth=96, minDesiredWidth=96, widthOverride=96}
  }
  VerticalBox_17 {widgetClassPath=&VerticalBox, parent=SizeBox_1, widget=VerticalBox_17, slot {class=SizeBoxSlot}}
  ScaleBox_1 {
   widgetClassPath=&ScaleBox
   parent=VerticalBox_17
   widget=ScaleBox_1
   properties {stretchDirection=DownOnly}
   slot {class=VerticalBoxSlot, properties {horizontalAlignment=HAlign_Center, padding=<ltrb 0 4 0 -2>}}
  }
  LevelHeaderText {
   widgetClassPath=&CommonTextBlock
   bIsVariable=true
   parent=ScaleBox_1
   widget=LevelHeaderText
   properties {
    justification=Center
    text=Level
    textTransformPolicy=ToUpper
    font {
     bForceMonospaced=false
     bMaterialIsStencil=false
     fontMaterial=&M_UI_BPDoor_Panel_Text_Header_Purchased
     fontObject=&BurbankSmall
     letterSpacing=0
     monospacedWidth=1
     size=13
     skewAmount=0
     typefaceFontName=Black
     outlineSettings {
      bApplyOutlineToDropShadows=false
      bMiteredCorners=false
      bSeparateFillAlpha=false
      outlineMaterial=null
      outlineSize=0
      outlineColor=<rgba 0 0 0 1>
     }
    }
    renderTransform {angle=0, scale=<1 1>, shear=<-12 3.56>, translation=<0 0>}
    shadowColorAndOpacity=<rgba 1 1 1 0.25>
    shadowOffset=<2 2>
   }
   slot {class=ScaleBoxSlot}
  }
  ScaleBox_0 {
   widgetClassPath=&ScaleBox
   parent=VerticalBox_17
   widget=ScaleBox_0
   properties {stretchDirection=DownOnly}
   slot {class=VerticalBoxSlot, properties {horizontalAlignment=HAlign_Center, verticalAlignment=VAlign_Center}}
  }
  LevelNumber {
   widgetClassPath=&CommonTextBlock
   bIsVariable=true
   parent=ScaleBox_0
   widget=LevelNumber
   properties {
    justification=Center
    lineHeightPercentage=0.8
    text=13
    font {
     bForceMonospaced=false
     bMaterialIsStencil=false
     fontMaterial=&M_UI_BPDoor_Panel_Text_Purchased
     fontObject=&Burbank
     letterSpacing=0
     monospacedWidth=1
     size=56
     skewAmount=0
     typefaceFontName=BurbankRegular-Black
     outlineSettings {
      bApplyOutlineToDropShadows=false
      bMiteredCorners=false
      bSeparateFillAlpha=true
      outlineMaterial=&M_UI_BPDoor_Panel_Text_Outline_Purchased
      outlineSize=2
      outlineColor=<rgba 1 1 1 1>
     }
    }
    margin=<ltrb 4 -16 4 0>
    renderTransform {angle=0, scale=<1 1>, shear=<0 3.5>, translation=<2 0>}
    shadowColorAndOpacity=<rgba 1 1 1 0.25>
    shadowOffset=<6 4>
   }
   slot {class=ScaleBoxSlot}
  }
  SizeBox_0 {
   widgetClassPath=&SizeBox
   parent=VerticalBox_17
   widget=SizeBox_0
   properties {
    bOverride_WidthOverride=true
    heightOverride=24
    widthOverride=56
    renderTransform {angle=0, scale=<1 0.75>, shear=<0 1>, translation=<0 0>}
    renderTransformPivot=<0.5 0>
   }
   slot {class=VerticalBoxSlot, properties {horizontalAlignment=HAlign_Center, verticalAlignment=VAlign_Top, padding=<ltrb 0 12 0 0>}}
  }
  FakeStatsGraphs_Top {
   widgetClassPath=/BRCosmetics/Athena/UI/Frontend/Spatial/Art/FakeStatsGraphs
   bIsVariable=true
   parent=SizeBox_0
   widget=FakeStatsGraphs_Top
   properties {colorAndOpacity=<rgba 1 0.7011 0.0999 0.2471>}
   slot {class=SizeBoxSlot, properties {verticalAlignment=VAlign_Top}}
  }
  SizeBox {
   widgetClassPath=&SizeBox
   parent=VerticalBox_17
   widget=SizeBox
   properties {bOverride_WidthOverride=true, heightOverride=24, widthOverride=56, renderTransformPivot=<0.5 0>}
   slot {class=VerticalBoxSlot, properties {horizontalAlignment=HAlign_Center, verticalAlignment=VAlign_Top, padding=<ltrb 0 22 0 0>}}
  }
  FakeStatsGraphs_Bottom {
   widgetClassPath=&FakeStatsGraphs
   bIsVariable=true
   parent=SizeBox
   widget=FakeStatsGraphs_Bottom
   properties {
    colorAndOpacity=<rgba 1 0.552 0.0999 0.1216>
    renderTransform {angle=0, scale=<1 0.4>, shear=<0 0>, translation=<0 0>}
    renderTransformPivot=<0.5 0>
   }
   slot {class=SizeBoxSlot, properties {verticalAlignment=VAlign_Top}}
  }
  Button_Navigation {widgetClassPath=&CommonButtonLegacy, bInherited=true, widget=null}
 ]
}
