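# Widget for the Battle Pass door in the spatial frontend that shows how much XP remains until the next level.
# The InitXpValues custom event sets XPAmount to XpToLevel minus CurrentXp and formats NextLevel into a localized "To Level" label.
# Two overlays in a sheared, slightly transparent horizontal box hold the texts over soft-glow material images, tilted in opposite directions for an angled look.
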WidgetBlueprint /BRCosmetics/Athena/UI/Frontend/Spatial/Art/BP_Door/BattlePass_Door_XP_Progress
graphs [EventGraph {nodes [
   n1 = InitXpValues!() -> (OutputDelegate, then, CurrentXp:int, XpToLevel:int, NextLevel:int) {is=title K2Node_CustomEvent}
   > "SetText (Text)"(
    self=Get XPAmount() -> (XPAmount:Common Text Object Reference)
    InText="To Text (Integer)"(Value=int - int(A=@n1.XpToLevel, B=@n1.CurrentXp) {is=title K2Node_CallFunction})
   )
   > "SetText (Text)"(
    self=Get NextLevelText() -> (NextLevelText:Common Text Object Reference)
    InText=Format Text!(Format="LOCTEXT(\"To Level {xp}\")", xp:int=@n1.NextLevel) -> (Result)
   )
   Event PreConstruct()
   Event Construct()
   Event Tick()
  ]}]
variables []
widgets {
 info {inheritedWidgetCount=0, parentClass=&UserWidget, rootWidgetClass=&HorizontalBox, widgetCount=11}
 widgets [
  All {
   widgetClassPath=&HorizontalBox
   widget=All
   properties {renderOpacity=0.8, renderTransform {angle=0, scale=<1 1>, shear=<-8 4.5>, translation=<0 0>}}
  }
  XP_Amount_Overlay {
   widgetClassPath=&Overlay
   parent=All
   widget=XP_Amount_Overlay
   properties {renderTransform {angle=1, scale=<1 1>, shear=<0 0>, translation=<0 0>}}
   slot {class=HorizontalBoxSlot, properties {verticalAlignment=VAlign_Center}}
  }
  Image_40 {
   widgetClassPath=&Image
   bIsVariable=true
   parent=XP_Amount_Overlay
   widget=Image_40
   properties {
    brush {
     bIsDynamicallyLoaded=false
     drawAs=Box
     imageType=NoImage
     mirroring=NoMirror
     resourceName=None
     resourceObject=/FortUILibrary/Styles/Border/Materials/M_Border_SoftGlow
     tiling=NoTile
     imageSize=<64 48>
     margin=<ltrb 0.5 0.5 0.5 0.5>
     outlineSettings {
      bUseBrushTransparency=false
      roundingType=HalfHeightRadius
      width=0
      color {colorUseRule=UseColor_Specified, specifiedColor=<rgba 0 0 0 0>}
      cornerRadii=<0 0 0 0>
     }
     tintColor {colorUseRule=UseColor_Specified, specifiedColor=<rgba 0.25 0.875 1 0.25>}
     uVRegion {bIsValid=false, max=<0 0>, min=<0 0>}
    }
    renderTransform {angle=0, scale=<1.3 1>, shear=<0 0>, translation=<0 0>}
   }
   slot {class=OverlaySlot, properties {horizontalAlignment=HAlign_Fill, verticalAlignment=VAlign_Fill}}
  }
  XPColor {
   widgetClassPath=&Border
   parent=XP_Amount_Overlay
   widget=XPColor
   properties {
    background {
     bIsDynamicallyLoaded=false
     drawAs=NoDrawType
     imageType=NoImage
     mirroring=NoMirror
     resourceName=None
     resourceObject=null
     tiling=NoTile
     imageSize=<32 32>
     margin=<ltrb 0 0 0 0>
     outlineSettings {
      bUseBrushTransparency=false
      roundingType=HalfHeightRadius
      width=0
      color {colorUseRule=UseColor_Specified, specifiedColor=<rgba 0 0 0 0>}
      cornerRadii=<0 0 0 0>
     }
     tintColor {colorUseRule=UseColor_Specified, specifiedColor=<rgba 1 1 1 1>}
     uVRegion {bIsValid=false, max=<0 0>, min=<0 0>}
    }
    contentColorAndOpacity=<rgba 0.6 0.9667 1 1>
    padding=<ltrb 0 0 0 0>
   }
   slot {class=OverlaySlot, properties {verticalAlignment=VAlign_Center}}
  }
  HorizontalBox_12 {
   widgetClassPath=&HorizontalBox
   parent=XPColor
   widget=HorizontalBox_12
   slot {class=BorderSlot, properties {padding=<ltrb 0 0 0 0>}}
  }
  XPAmount {
   widgetClassPath=&CommonTextBlock
   bIsVariable=true
   parent=HorizontalBox_12
   widget=XPAmount
   properties {
    text=1358
    font {
     bForceMonospaced=false
     bMaterialIsStencil=false
     fontMaterial=null
     fontObject=&BurbankSmall
     letterSpacing=0
     monospacedWidth=1
     size=16
     skewAmount=0
     typefaceFontName=Black
     outlineSettings {
      bApplyOutlineToDropShadows=false
      bMiteredCorners=false
      bSeparateFillAlpha=false
      outlineMaterial=null
      outlineSize=0
      outlineColor=<rgba 0 0 0 1>
     }
    }
    shadowColorAndOpacity=<rgba 1 1 1 0.158>
    shadowOffset=<-4 2>
   }
   slot {class=HorizontalBoxSlot}
  }
  CommonTextBlock_102 {
   widgetClassPath=&CommonTextBlock
   parent=HorizontalBox_12
   widget=CommonTextBlock_102
   properties {
    text=XP
    textTransformPolicy=ToUpper
    font {
     bForceMonospaced=false
     bMaterialIsStencil=false
     fontMaterial=null
     fontObject=&BurbankSmall
     letterSpacing=0
     monospacedWidth=1
     size=16
     skewAmount=0
     typefaceFontName=Black
     outlineSettings {
      bApplyOutlineToDropShadows=false
      bMiteredCorners=false
      bSeparateFillAlpha=false
      outlineMaterial=null
      outlineSize=0
      outlineColor=<rgba 0 0 0 1>
     }
    }
    shadowColorAndOpacity=<rgba 1 1 1 0.15>
    shadowOffset=<-4 2>
   }
   slot {class=HorizontalBoxSlot, properties {horizontalAlignment=HAlign_Left, verticalAlignment=VAlign_Center}}
  }
  Next_Level_Overlay {
   widgetClassPath=&Overlay
   parent=All
   widget=Next_Level_Overlay
   properties {renderTransform {angle=-1, scale=<1 1>, shear=<0 0>, translation=<0 0>}, renderTransformPivot=<0 0.5>}
   slot {class=HorizontalBoxSlot, properties {verticalAlignment=VAlign_Center}}
  }
  Image {
   widgetClassPath=&Image
   bIsVariable=true
   parent=Next_Level_Overlay
   widget=Image
   properties {
    brush {
     bIsDynamicallyLoaded=false
     drawAs=Box
     imageType=NoImage
     mirroring=NoMirror
     resourceName=None
     resourceObject=&M_Border_SoftGlow
     tiling=NoTile
     imageSize=<64 48>
     margin=<ltrb 0.5 0.5 0.5 0.5>
     outlineSettings {
      bUseBrushTransparency=false
      roundingType=HalfHeightRadius
      width=0
      color {colorUseRule=UseColor_Specified, specifiedColor=<rgba 0 0 0 0>}
      cornerRadii=<0 0 0 0>
     }
     tintColor {colorUseRule=UseColor_Specified, specifiedColor=<rgba 0.2 0.7333 1 0.25>}
     uVRegion {bIsValid=false, max=<0 0>, min=<0 0>}
    }
    renderTransform {angle=0, scale=<1.3 1>, shear=<0 0>, translation=<0 0>}
   }
   slot {class=OverlaySlot, properties {horizontalAlignment=HAlign_Fill, verticalAlignment=VAlign_Fill}}
  }
  NextLevelColor {
   widgetClassPath=&Border
   parent=Next_Level_Overlay
   widget=NextLevelColor
   properties {
    background {
     bIsDynamicallyLoaded=false
     drawAs=NoDrawType
     imageType=NoImage
     mirroring=NoMirror
     resourceName=None
     resourceObject=null
     tiling=NoTile
     imageSize=<32 32>
     margin=<ltrb 0 0 0 0>
     outlineSettings {
      bUseBrushTransparency=false
      roundingType=HalfHeightRadius
      width=0
      color {colorUseRule=UseColor_Specified, specifiedColor=<rgba 0 0 0 0>}
      cornerRadii=<0 0 0 0>
     }
     tintColor {colorUseRule=UseColor_Specified, specifiedColor=<rgba 1 1 1 1>}
     uVRegion {bIsValid=false, max=<0 0>, min=<0 0>}
    }
    contentColorAndOpacity=<rgba 0.5 0.8333 1 1>
    padding=<ltrb 0 0 0 0>
   }
   slot {class=OverlaySlot, properties {verticalAlignment=VAlign_Center, padding=<ltrb 4 0 4 0>}}
  }
  NextLevelText {
   widgetClassPath=&CommonTextBlock
   bIsVariable=true
   parent=NextLevelColor
   widget=NextLevelText
   properties {
    text=To Level 14
    textTransformPolicy=ToUpper
    font {
     bForceMonospaced=false
     bMaterialIsStencil=false
     fontMaterial=null
     fontObject=&BurbankSmall
     letterSpacing=0
     monospacedWidth=1
     size=12
     skewAmount=0
     typefaceFontName=Black
     outlineSettings {
      bApplyOutlineToDropShadows=false
      bMiteredCorners=false
      bSeparateFillAlpha=false
      outlineMaterial=null
      outlineSize=0
      outlineColor=<rgba 0 0 0 1>
     }
    }
    renderTransform {angle=0, scale=<1 1>, shear=<0 0>, translation=<0 2>}
    shadowColorAndOpacity=<rgba 1 1 1 0.15>
    shadowOffset=<-3 2>
   }
   slot {class=BorderSlot, properties {padding=<ltrb 0 0 0 0>}}
  }
 ]
}
